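# Additive material for the laser beams on the Laserstar SE wheel cosmetic, flagged for skeletal meshes.
# CustomColor times BeamEmissiveStrength tints either a procedural, width-controlled band across V or, under the CheapEffect switch, a laser texture.
# The cheap texture flickers by flipping U with the sign of a timed sine.
# The full version also builds an animated local-space wobble from scrolling sine waves, masked toward mid-U, which CheapEffect replaces with a constant.
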Material /VehicleCosmetics/Wheels/LaserstarSE/Materials/M_Laserstar_Laserbeams_fx
graph {
 properties {
  bAutomaticallySetUsageInEditor=false
  bUsedWithMeshDeformer=true
  bUsedWithSkeletalMesh=true
  blendMode=BLEND_Additive
  usedShadingModels=Default Lit
 }
 nodes [
  n1 = Add(A=TextureCoordinate()) {constB=-0.5}
  n2 = OneMinus(None=Abs(None=ComponentMask(None=@n1) {r=true}))
  n3 = Clamp(None=Multiply(A=Subtract(A=@n2) {constB=0.5}) {constB=3})
  MaterialFunctionCall(R=@n3, G=@n3, B=@n3) {materialFunction=&AppendMany}
  n4 = Multiply(
   A=VectorParameter() {parameterName=CustomColor, defaultValue=<rgba 0 1 0.2195 1>}.RGB
   B=ScalarParameter() {defaultValue=3, parameterName=BeamEmissiveStrength}
  )
  n5 = Sine(None=Multiply(A=ScalarParameter() {defaultValue=5, parameterName=LODBeam speed}, B=Time()))
  n6 = TextureCoordinate()
  n7 = AppendVector(A=Multiply(A=Sign(None=@n5), B=ComponentMask(None=@n6) {r=true}), B=ComponentMask(None=@n6) {g=true})
  n8 = TextureCoordinate()
  n9 = Abs(None=ComponentMask(None=Add(A=@n8) {constB=-0.5}) {g=true})
  n10 = Multiply(A=Subtract(A=OneMinus(None=@n9), B=ScalarParameter() {defaultValue=0.6, parameterName=BeamWidth}) {constB=0.6}) {constB=4}
  StaticSwitchParameter(
   True=Multiply(A=@n4, B=TextureSample(UVs=@n7) {texture=/VehicleCosmetics/Wheels/LaserstarSE/Textures/Laser_250SE}.R)
   False=Multiply(A=@n4, B=Clamp(None=@n10))
  ) {parameterName=CheapEffect}
  n11 = TextureCoordinate() {uTiling=6, vTiling=6}
  n12 = Multiply(A=ScalarParameter() {defaultValue=-6, parameterName=speed}, B=Time())
  n13 = Sine(None=Add(A=ComponentMask(None=@n11) {r=true}, B=@n12)) {period=1.5}
  n14 = Add(A=ComponentMask(None=@n11) {g=true}, B=@n12)
  n15 = Sine(None=@n14) {period=2}
  n16 = Cosine(None=@n14)
  n17 = ConstantBiasScale(None=Transform(None=MaterialFunctionCall(R=@n13, G=@n15, B=@n16) {materialFunction=&AppendMany}.RGB) {transformSourceType=TRANSFORMSOURCE_Local})
  n18 = ScalarParameter() {defaultValue=2, parameterName=wobbleamount}
  n19 = Transform(None=MaterialFunctionCall(R=@n18, G=@n18, B=@n18) {materialFunction=&AppendMany}.RGB) {transformSourceType=TRANSFORMSOURCE_Local}
  StaticSwitchParameter(True=Constant(), False=Multiply(A=@n3, B=Multiply(A=@n17, B=@n19))) {parameterName=CheapEffect}
 ]
}
